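# Niagara effect for the HauntKoi back bling cosmetic that draws a glowing black hole with accretion disks.
# All emitters burst at a shared System.SpawnPosition; BlackHole layers three sprites and a red-orange light, while Accretion_Disks uses two mesh renderers.
# Custom per-layer color, size and camera-offset attributes appear to let several renderers share each particle.
# BigGlow repeats a soft red sprite that fades in and out; the rotating mesh Wisps emitter is disabled.
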NiagaraSystem /BRCosmetics/Accessories/FORT_Backpacks/Backpack_M_MED_HauntKoi/FX/NS_Backpack_HauntKoi
dependencies {
 usedDataInterfaces [&NiagaraDataInterfaceCurve, &NiagaraDataInterfaceVectorCurve]
 usedDynamicInputs []
 usedModules [
  &SystemState
  &self:SystemSpawnScript.SystemScriptSource.SystemScriptGraph.NiagaraNodeAssignment_0.SetVariables_#1
 ]
}
emitters {
 Accretion_Disks {
  emitterUpdateScript [
   &EmitterState(
    Inactive Response=::NewEnumerator1
    Life Cycle Mode=::NewEnumerator1
    Loop Behavior=::NewEnumerator1
    Loop Duration Mode=::NewEnumerator1
    UseLoopDelay=false
   )
   &SpawnBurst_Instantaneous(Use Spawn Probability=false, UseLoopCountLimit=true)
  ]
  particleSpawnScript [
   &InitializeParticle(
    Mesh Scale Mode=::NewEnumerator4
    Position=$System.SpawnPosition
    Position Mode=::NewEnumerator1
    UsePositionOffset=false
    Write Material Random=true
   )
   /SetVariables(Particles.Color Disk Fast:LinearColor=<rgba 1 1 1 2>) {isSetParameters=true}
   /SetVariables(Particles.Color Disk Slow:LinearColor=<rgba 1 1 1 2.5>) {isSetParameters=true}
   &CameraOffset(Camera Offset Amount=-2)
   &DynamicMaterialParameters(
    Index 0 Param 2=3
    Module.Param0WriteEnabled=true
    Module.Param1WriteEnabled=true
    Module.Param2WriteEnabled=true
    Module.Param3WriteEnabled=true
   ) {enabled=false}
  ]
  rendererClasses [&NiagaraMeshRendererProperties]
  renderers [
   {rendererClass=&NiagaraMeshRendererProperties, rendererIndex=0}
   {rendererClass=&NiagaraMeshRendererProperties, rendererIndex=1}
  ]
 }
 BigGlow {
  emitterUpdateScript [
   &EmitterState(Life Cycle Mode=::NewEnumerator1, Loop Duration=3, UseLoopDelay=true)
   &SpawnBurst_Instantaneous(Use Spawn Probability=false, UseLoopCountLimit=false)
  ]
  particleSpawnScript [
   &InitializeParticle(
    Lifetime=4
    Position=$System.SpawnPosition
    Position Mode=::NewEnumerator1
    Sprite Size=<90 90>
    Sprite Size Mode=::NewEnumerator4
    UsePositionOffset=false
    Write Material Random=true
   )
   &Color(Color=<rgba 1 0 0 0.25>)
  ]
  particleUpdateScript [
   &ParticleState()
   &ScaleColor(
    Scale Alpha=&FloatFromCurve(FloatCurve=DataInterface() {propertyValues {Curve {keys [<at 0 0 RCIM_Cubic>, <at 0.25 1>, {arriveTangent=-1.3333, interpMode=RCIM_Cubic, leaveTangent=-1.3333, time=0.75, value=1}, <at 1 0>]}}})
    ScaleA=true
    ScaleRGB=true
   )
  ]
  rendererClasses [&NiagaraSpriteRendererProperties]
  renderers [{rendererClass=&NiagaraSpriteRendererProperties}]
 }
 BlackHole {
  emitterUpdateScript [
   &EmitterState(Life Cycle Mode=::NewEnumerator1, Loop Duration=1, UseLoopDelay=true)
   &SpawnBurst_Instantaneous(Use Spawn Probability=false, UseLoopCountLimit=true)
  ]
  particleSpawnScript [
   &InitializeParticle(
    Position=$System.SpawnPosition
    Position Mode=::NewEnumerator1
    Sprite Size=<1 1>
    Sprite Size Mode=::NewEnumerator4
    UsePositionOffset=false
    Write Material Random=true
   )
   /SetVariables(Particles.Color Black Hole:LinearColor=<rgba 1 1 1 0.5>) {isSetParameters=true}
   /SetVariables(Particles.Color Kaleidoscope:LinearColor=<rgba 1 0 0 0.5>) {isSetParameters=true}
   /SetVariables(Particles.Color Outer Wisps:LinearColor=<rgba 1 0 0 100>) {isSetParameters=true}
   /SetVariables(Particles.Sprite Size BlackHole:Vector2f=<26 26>) {isSetParameters=true}
   /SetVariables(Particles.Sprite Size Kaleidoscope:Vector2f=<50 50>) {isSetParameters=true}
   /SetVariables(Particles.Sprite Size Outer Wisps:Vector2f=<50 50>) {isSetParameters=true}
   /SetVariables(Particles.Cam Offset Kaleidoscope=15) {isSetParameters=true}
   /SetVariables(Particles.Cam Offset Outer Wisps=5) {isSetParameters=true}
  ]
  particleUpdateScript [&FN_Light_Attributes(
    Color=<rgba 0.651 0.0268 0 1>
    Create Color Attribute=true
    Create Exponent Attribute=true
    Create Radius Attribute=true
    Enable Particle Light Enable=true
    Exponent=3
    Light Enabled=true
    Radius=75
   )]
  rendererClasses [&NiagaraSpriteRendererProperties, &NiagaraLightRendererProperties]
  renderers [
   {rendererClass=&NiagaraSpriteRendererProperties}
   {rendererClass=&NiagaraSpriteRendererProperties, rendererIndex=1}
   {rendererClass=&NiagaraSpriteRendererProperties, rendererIndex=2}
   {rendererClass=&NiagaraLightRendererProperties, rendererIndex=3}
  ]
 }
 Wisps {
  bEnabled=false
  emitterUpdateScript [
   &EmitterState(Life Cycle Mode=::NewEnumerator1, Loop Duration=1, UseLoopDelay=true)
   &SpawnBurst_Instantaneous(Use Spawn Probability=false, UseLoopCountLimit=false)
  ]
  particleSpawnScript [
   /SetVariables(Particles.MeshOrientation=&MakeQuatFromAxisAngle(Axis=<1 1 -1.0615>, AngleInDegrees=45, NormalizeAxis=true)) {isSetParameters=true}
   /SetVariables(Particles.SpawnRotationVector=<0 0 0.25>) {isSetParameters=true}
   &InitializeParticle(
    Lifetime=&RandomRangeFloat(Minimum=1.1, Maximum=1.25)
    Mesh Scale=&RandomRangeVector(Minimum=<0.9 0.85 0.75>, Maximum=<0.95 0.9 0.8>)
    Mesh Scale Mode=::NewEnumerator4
    Position=$System.SpawnPosition
    Position Mode=::NewEnumerator1
    UsePositionOffset=false
    Write Material Random=true
   )
   &Color(Color=&MakeLinearColorFromVectorAndFloat("Vector (RGB)"=<1 0.02 0>, "Float (Alpha)"=&RandomRangeFloat(Minimum=0.5, Maximum=0.8)))
   &DynamicMaterialParameters(
    Index 0 Param 1=&RandomRangeFloat(Minimum=0, Maximum=1)
    Index 0 Param 3=1
    Module.Param0WriteEnabled=true
    Module.Param1WriteEnabled=true
    Module.Param2WriteEnabled=true
    Module.Param3WriteEnabled=true
   )
  ]
  particleUpdateScript [
   &ParticleState()
   &UpdateMeshOrientation(Rotation Rate=2, Rotation Vector=$Particles.SpawnRotationVector)
   &UpdateMeshOrientation(
    Facing Direction=<90 0 0>
    Facing Interpolation=false
    Facing Interpolation Rate=50
    Orientation Method=::NewEnumerator1
    Override Quat=false
    TransformFacing=false
    TransformUp=false
   ) {module=UpdateMeshOrientation001}
   &ScaleColor(
    Scale Alpha=&FloatFromCurve(FloatCurve=DataInterface() {propertyValues {Curve {keys [<at 0 0 RCIM_Cubic>,<at 0.5 1 RCIM_Cubic>,<at 1 0 RCIM_Cubic>]}}})
    Scale RGB=&VectorFromCurve(VectorCurve=DataInterface() {propertyValues {
       XCurve {keys [<at 0 0.2276 RCIM_Cubic>,<at 1 0.7259 RCIM_Cubic>]}
       YCurve {keys [<at 0 2.7754 RCIM_Cubic>,<at 1 2.4246 RCIM_Cubic>]}
       ZCurve {keys [<at 0 6 RCIM_Cubic>,<at 1 6 RCIM_Cubic>]}
      }})
    ScaleA=true
    ScaleRGB=true
   )
  ]
  rendererClasses [&NiagaraMeshRendererProperties]
  renderers [{rendererClass=&NiagaraMeshRendererProperties, rendererIndex=0}]
 }
}
summary {
 systemName=NS_Backpack_HauntKoi
 emitters [
  {emitterName=BigGlow, rendererClasses [&NiagaraSpriteRendererProperties]}
  {emitterName=BlackHole, rendererClasses [&NiagaraSpriteRendererProperties, &NiagaraLightRendererProperties]}
  {emitterName=Accretion_Disks, rendererClasses [&NiagaraMeshRendererProperties]}
  {bEnabled=false, emitterName=Wisps, rendererClasses [&NiagaraMeshRendererProperties]}
 ]
 userVariables []
}
